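# Unlit masked material for Spark graffiti screens, layering a panning background, two graffiti textures and a splatter.
# Layers composite through masks from the textures, tinted by stage colors from MF_MojoStageColors, with a parameter blending toward full texture color.
# Beat and measure lifetimes from an audio parameter collection drive curve-atlas remaps for background, splatter and global reactivity.
# A panning noise reroute is also defined, and the composite feeds MF_Sparks_Screen_Details as emissive.
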Material /SparksMaps/Effects/Screens/Graffiti/M_Spark_Screens_Graffiti
graph {
 properties {bAutomaticallySetUsageInEditor=false, bUsedWithNanite=true, blendMode=BLEND_Masked, shadingModel=MSM_Unlit, usedShadingModels=Unlit}
 nodes [
  n1 = NamedRerouteUsage()
  n2 = NamedRerouteUsage()
  n3 = Multiply(
   A=TextureCoordinate()
   B=AppendVector(A=ScalarParameter() {defaultValue=0.8, parameterName=Tile BG}, B=ScalarParameter() {defaultValue=0.8, parameterName=Tile BG Y})
  )
  n4 = AppendVector(A=ScalarParameter() {defaultValue=1, parameterName=Offset BG}, B=ScalarParameter() {defaultValue=1, parameterName=Offset BG B})
  n5 = ScalarParameter() {defaultValue=-0.0025, parameterName=Speed BG 1}
  n6 = ScalarParameter() {defaultValue=0.005, parameterName=Speed BG 2}
  n7 = TextureSample(UVs=Panner(Coordinate=Add(A=@n3, B=@n4), Speed=AppendVector(A=@n5, B=@n6)) {speedX=0.001, speedY=0.001}) {texture=&/T_Sparks_Screen_Graffiti_BG}
  n8 = Add(
   A=Multiply(A=MaterialFunctionCall!(CustomUPosition) {materialFunction=/SparksStageShow/Systems/StageManager/FXManager/MaterialFunctions/MF_MojoStageColors}.Tertiary) {constB=0.05}
   B=Multiply(
    A=MaterialFunctionCall() {materialFunction=&MF_MojoStageColors}.Tertiary
    B=@n7.R
   )
  )
  n9 = AppendVector(A=ScalarParameter() {parameterName=Offset Graffiti 1A}, B=ScalarParameter() {parameterName=Offset Graffiti 1B})
  n10 = ScalarParameter() {defaultValue=1, parameterName=GlobalScaleX}
  n11 = ScalarParameter() {defaultValue=1, parameterName=GlobalScaleY}
  n12 = Multiply(
   A=TextureCoordinate()
   B=Multiply(A=AppendVector(A=@n10, B=@n11), B=ScalarParameter() {defaultValue=1.2, parameterName=Tile Foreground Graffiti}) {constB=1.2}
  )
  n13 = ScalarParameter() {defaultValue=0.0075, parameterName=Speed 1A}
  n14 = ScalarParameter() {defaultValue=0.005, parameterName=Speed 1B}
  n15 = TextureSample(UVs=Panner(Coordinate=Add(A=@n9, B=@n12), Speed=AppendVector(A=@n13, B=@n14))) {texture=&/T_Sparks_Screen_Graffiti_1}
  n16 = Multiply(
   A=MaterialFunctionCall() {materialFunction=&MF_MojoStageColors}.Secondary
   B=@n15.R
  )
  n17 = AppendVector(A=ScalarParameter() {parameterName=Offset Graffiti 2A}, B=ScalarParameter() {parameterName=Offset Graffiti 2B})
  n18 = ScalarParameter() {defaultValue=0.005, parameterName=Speed 2A}
  n19 = ScalarParameter() {defaultValue=0.0125, parameterName=Speed 2B}
  n20 = TextureSample(UVs=Panner(Coordinate=Add(A=@n17, B=@n12), Speed=AppendVector(A=@n18, B=@n19))) {texture=&/T_Sparks_Screen_Graffiti_2}
  n21 = NamedRerouteUsage()
  n22 = TextureCoordinate() {uTiling=2, vTiling=2}
  n23 = ScalarParameter() {defaultValue=1.4, parameterName=SplatterTile}
  n24 = ScalarParameter() {defaultValue=0.01, parameterName=Speed SplatterA}
  n25 = ScalarParameter() {defaultValue=0.02, parameterName=Speed SplatterB}
  n26 = TextureSample(UVs=Panner(Coordinate=Multiply(A=@n22, B=@n23), Speed=AppendVector(A=@n24, B=@n25)) {speedY=0.2}) {texture=&/T_Sparks_Screen_Graffiti_Splatter}
  n27 = Saturate(None=Multiply(A=Multiply(A=@n21, B=@n26.R), B=Constant() {r=5}))
  n28 = Multiply(
   A=Add(A=@n20.R, B=@n27)
   B=MaterialFunctionCall() {materialFunction=&MF_MojoStageColors}.Primary
  )
  n29 = Constant() {r=5}
  n30 = Add(A=Saturate(None=Multiply(A=@n20.R, B=@n29)), B=@n27)
  n31 = LinearInterpolate(A=@n15.RGB, B=Add(A=@n20.RGB, B=Multiply(A=@n21, B=@n26.RGB)), Alpha=@n30)
  n32 = Saturate(None=Multiply(A=Desaturation(None=@n31), B=Constant() {r=10}) {constB=1000})
  n33 = LinearInterpolate(
   A=LinearInterpolate(A=Multiply(A=@n2, B=@n8), B=LinearInterpolate(A=@n16, B=@n28, Alpha=@n30), Alpha=@n32) {desc=Stage Color}
   B=LinearInterpolate(A=@n7.RGB, B=@n31, Alpha=@n32) {desc=Full Color}
   Alpha=ScalarParameter() {parameterName=Stage or Full Color}
  )
  MaterialFunctionCall!(Emissive=Multiply(A=Multiply(A=@n1, B=@n33), B=NamedRerouteUsage()), Scanlines Tile) {materialFunction=/SparksMaps/Effects/Screens/Materials/MF/MF_Sparks_Screen_Details}
  n34 = TextureCoordinate()
  n35 = AppendVector(A=Constant() {r=0.75}, B=Constant() {r=0.75})
  n36 = ScalarParameter() {defaultValue=0.1, parameterName=SpeedNoiseA}
  n37 = ScalarParameter() {defaultValue=0.1, parameterName=SpeedNoiseB}
  n38 = TextureSample(UVs=Panner(Coordinate=Multiply(A=@n34, B=@n35), Speed=AppendVector(A=@n36, B=@n37)) {speedX=0.1, speedY=0.1}) {samplerType=SAMPLERTYPE_LinearGrayscale, texture=/Game/Global/Textures/Default/Noise/EqualizerComponents_v2_Noise}
  NamedRerouteDeclaration(None=LinearInterpolate(Alpha=Saturate(None=@n38.R))) {name=NOise, nodeColor=<rgba 0.4531 1 0 1>}
  n39 = CurveAtlasRowParameter(CurveTime=CollectionParameter() {collection=/FMAudioReactiveData/AudioAnalysis/MPC_Sparks_Audio_Parameters, parameterName=BeatLifetime}) {
   atlas=/SparksMaps/Effects/Screens/Curves/CurveAtlas_Audio_Beat_A
   curve=/SparksMaps/Effects/Screens/Curves/Curve_Audio_Beat_Remap_001
   parameterName=ReactivityCurve
  }
  NamedRerouteDeclaration(None=MaterialFunctionCall(
    Input=@n39
    Input Low=Constant()
    Input High=Constant() {r=1}
    Target Low=ScalarParameter() {defaultValue=0.3, parameterName=BG Beat Low}
    Target High=Constant() {r=1}
   ) {materialFunction=&RemapValueRange}.Result) {name=ReactivityBG, nodeColor=<rgba 0 1 0.2266 1>}
  n40 = CurveAtlasRowParameter(CurveTime=CollectionParameter() {collection=&MPC_Sparks_Audio_Parameters, parameterName=MeasureLifetime}) {
   atlas=&CurveAtlas_Audio_Beat_A
   curve=&Curve_Audio_Beat_Remap_001
   parameterName=ReactivityCurve
  }
  NamedRerouteDeclaration(None=MaterialFunctionCall(
    Input=@n40
    Input Low=Constant()
    Input High=Constant() {r=1}
    Target Low=ScalarParameter() {defaultValue=0.75, parameterName=BeatLowGlobal}
    Target High=Constant() {r=1}
   ) {materialFunction=&RemapValueRange}.Result) {name=ReactivityGlobal, nodeColor=<rgba 0 0 0 1>}
  n41 = CurveAtlasRowParameter(CurveTime=CollectionParameter() {collection=&MPC_Sparks_Audio_Parameters, parameterName=BeatLifetime}) {
   atlas=&CurveAtlas_Audio_Beat_A
   curve=&Curve_Audio_Beat_Remap_001
   parameterName=ReactivityCurve
  }
  NamedRerouteDeclaration(None=MaterialFunctionCall(
    Input=@n41
    Input Low=Constant()
    Input High=Constant() {r=1}
    Target Low=ScalarParameter() {defaultValue=0.15, parameterName=SplatterBeatLow}
    Target High=Constant() {r=1}
   ) {materialFunction=&RemapValueRange}.Result) {name=SplatterReactivity, nodeColor=<rgba 1 0.2344 0 1>}
 ]
}
